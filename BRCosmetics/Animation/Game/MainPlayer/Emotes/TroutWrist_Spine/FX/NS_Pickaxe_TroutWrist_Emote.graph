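# Niagara ribbon trail system for the TroutWrist pickaxe emote, tracing motion from skeletal mesh bones or sockets.
# Active emitter Em_Trail spawns three short-lived AnimTrails with per-group position offsets and a user-driven velocity, drawn by three ribbon renderers.
# Curves shape ribbon width and alpha over life, while system time-of-day color and intensity scale the warm tint.
# Two similar emitters, Em_Trail001 and Em_Trail002, are disabled variants with longer lifetimes.
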NiagaraSystem /BRCosmetics/Animation/Game/MainPlayer/Emotes/TroutWrist_Spine/FX/NS_Pickaxe_TroutWrist_Emote
dependencies {
 usedDataInterfaces [
  &NiagaraDataInterfaceVectorCurve
  &NiagaraDataInterfaceArrayFloat3
  &NiagaraDataInterfaceCurve
  &NiagaraDataInterfaceSkeletalMesh
  &NiagaraDataInterfaceColorCurve
 ]
 usedDynamicInputs [
  &self:SectionIndexFloatSelector_02
  &ColorFromCurve
  &FNTimeOfDayScalar
 ]
 usedModules [
  &self:SystemSpawnScript.SystemScriptSource.SystemScriptGraph.NiagaraNodeAssignment_6.SetVariables_#1
  &SystemState
  &self:SystemSpawnScript.SystemScriptSource.SystemScriptGraph.NiagaraNodeAssignment_2.SetVariables_#2
  &self:SystemSpawnScript.SystemScriptSource.SystemScriptGraph.NiagaraNodeAssignment_0.SetVariables_#3
 ]
}
emitters {
 Em_Trail {
  emitterUpdateScript [
   &EmitterState()
   &FN_SpawnAnimTrail(Number of Anim Trails=3, Spawn Interval=0, Spawn Mode:ENiagaraSkeletalSpawning=::NewEnumerator0) {module=FN_SpawnAnimTrail001}
  ]
  particleSpawnScript [
   /Game/Effects/Niagara/Modules/AnimTrail/FN_InitializeAnimTrail/FN_InitializeAnimTrailBoneIndecies()
   /Game/Effects/Niagara/Modules/AnimTrail/FN_InitializeAnimTrailBoneIndecies_v2/FN_InitializeAnimTrailBoneIndecies(AnimTrail Selection=1, First Bone or Socket Index=1) {module=FN_InitializeAnimTrailBoneIndecies001}
   &FN_InitializeAnimTrailBoneIndecies(AnimTrail Selection=2, First Bone or Socket Index=2, Second Bone or Socket Index=2) {module=FN_InitializeAnimTrailBoneIndecies002}
   &FN_InitializeAnimTrail(
    AnimTrailWidth=2
    Color=&MultiplyLinearColor(
     Linear Color=<rgba 1 0.8509 0.5877 1>
     Scale Factor=&MakeVector4_FromLinearColor(LinearColor=$System.ToD Color)
    )
    Lifetime=0.16
    Position Offset=&MaskVectorBySpawnGroup(Masked Vector=<32 0 -8>, Passthrough Vector=<32 0 8>, Spawn Group Mask=$Particles.SpawnGroup)
    SkeletalMesh=$System.SkeletalMesh
   )
   /SetVariables(Particles.Position=&Add_Vector(A=Linked() {linkedVariable {name=Particles.Position, type=NiagaraPosition}}, B=<-32 0 0>)) {isSetParameters=true, module=SetVariables_#4}
   &SkeletalMeshLocation(
    Bone / Socket Index=$Particles.SpawnGroup
    Bone Sampling Mode=::NewEnumerator4
    MaskSpawnGroup=false
    Sampled Position Offset=&Add_Vector(
     A=&VectorFromCurve(
      VectorCurve=DataInterface() {propertyValues {XCurve {keys [<at 0 -25>,<at 1 -20>]}, YCurve {keys [<at 0 1>,<at 1 0>]}, ZCurve {keys [<at 0 1>,<at 1 0>]}}}
      Scale Curve=<0 0 0>
     )
     B=&SelectVectorFromArray(Direct Array Index=$Particles.SpawnGroup, Vector Selection Array=DataInterface() {propertyValues {FloatData [<32 0 -32>, <-32 0 -32>]}})
    )
    Skeletal Mesh=$System.SkeletalMesh
    Velocity Sampling=::NewEnumerator0
    VelocitySpeedLimit=false
   )
   /SetVariables(Particles.Position=&AddVectorToPosition(Position=$Particles.Position, Vector=<0 0 0>)) {isSetParameters=true}
   /SetVariables(
    Particles.PositionOffset1:NiagaraPosition=&AddVectorToPosition(Position=$Particles.Position, Vector=<0 0 8>)
    Particles.PositionOffset2:NiagaraPosition=&AddVectorToPosition(Position=$Particles.Position, Vector=<0 0 -8>)
   ) {isSetParameters=true, module=SetVariables_#5}
   &AddVelocity(Velocity=<-1 0 1>) {enabled=false}
   &AddVelocity(Velocity=$User.Ribbon camera face) {module=AddVelocity001}
   &ApplyInitialForces()
  ]
  particleUpdateScript [
   &FN_UpdateAnimTrail(SkeletalMesh=$System.SkeletalMesh)
   &ScaleRibbonWidth(Ribbon Width Scale=&FloatFromCurve(
     FloatCurve=DataInterface() {propertyValues {Curve {keys [
         {arriveTangent=32.7595, interpMode=RCIM_Cubic, leaveTangent=32.7595, tangentMode=RCTM_User, time=0, value=0}
         <at 0.01 1 RCIM_Cubic>
         {arriveTangent=-1.4352, tangentMode=RCTM_User, time=1, value=0}
        ]}}}
     Scale Curve=2
    ))
   &ScaleColor(
    Scale Alpha=&Multiply_Float(
     A=&FloatFromCurve(FloatCurve=DataInterface() {propertyValues {Curve {keys [
          {arriveTangent=6.2556, interpMode=RCIM_Cubic, leaveTangent=6.2556, tangentMode=RCTM_User, time=0, value=0}
          {arriveTangent=-0.5296, interpMode=RCIM_Cubic, leaveTangent=-0.5296, tangentMode=RCTM_User, time=0.1, value=0.25}
          <at 1 0>
         ]}}})
     B=1
    )
    Scale RGB=&VectorFromFloat(Value=$System.ToD Intensity)
    ScaleA=true
    ScaleRGB=true
   )
   &ParticleState()
  ]
  rendererClasses [&NiagaraRibbonRendererProperties]
  renderers [
   {rendererClass=&NiagaraRibbonRendererProperties}
   {rendererClass=&NiagaraRibbonRendererProperties, rendererIndex=1}
   {rendererClass=&NiagaraRibbonRendererProperties, rendererIndex=2}
  ]
 }
 Em_Trail001 {
  bEnabled=false
  emitterUpdateScript [
   &EmitterState()
   &FN_SpawnAnimTrail(Number of Anim Trails=3, Spawn Interval=0, Spawn Mode:ENiagaraSkeletalSpawning=::NewEnumerator0) {module=FN_SpawnAnimTrail001}
  ]
  particleSpawnScript [
   /Game/Effects/Niagara/Modules/AnimTrail/FN_InitializeAnimTrail/FN_InitializeAnimTrailBoneIndecies()
   /Game/Effects/Niagara/Modules/AnimTrail/FN_InitializeAnimTrailBoneIndecies_v2/FN_InitializeAnimTrailBoneIndecies(AnimTrail Selection=1, Second Bone or Socket Index=0) {module=FN_InitializeAnimTrailBoneIndecies001}
   &FN_InitializeAnimTrailBoneIndecies(AnimTrail Selection=2, Second Bone or Socket Index=0) {module=FN_InitializeAnimTrailBoneIndecies002}
   &SkeletalMeshLocation(
    Bone / Socket Index=$Particles.SpawnGroup
    Bone Sampling Mode=::NewEnumerator4
    MaskSpawnGroup=false
    Sampled Position Offset=<0 60 0>
    Skeletal Mesh=$System.SkeletalMesh
   )
   &FN_InitializeAnimTrail(
    AnimTrailWidth=2
    Color=&MultiplyLinearColor(
     Linear Color=<rgba 1 0.8509 0.5877 1>
     Scale Factor=&MakeVector4_FromLinearColor(LinearColor=$System.ToD Color)
    )
    Lifetime=1
    Position Offset=&MaskVectorBySpawnGroup(Masked Vector=<32 0 -8>, Passthrough Vector=<32 0 8>, Spawn Group Mask=$Particles.SpawnGroup)
    SkeletalMesh=$System.SkeletalMesh
   )
   /SetVariables(Particles.Position=&Add_Vector(A=Linked() {linkedVariable {name=Particles.Position, type=NiagaraPosition}}, B=<0 0 0>)) {isSetParameters=true, module=SetVariables_#4}
   /SetVariables(
    Particles.PositionOffset1:NiagaraPosition=&AddVectorToPosition(Position=$Particles.Position, Vector=<0 0 0>)
    Particles.PositionOffset2:NiagaraPosition=&AddVectorToPosition(Position=$Particles.Position, Vector=<0 0 0>)
   ) {isSetParameters=true, module=SetVariables_#5}
   &AddVelocity(Velocity=<-1 0 1>) {enabled=false}
   &ApplyInitialForces() {enabled=false}
  ]
  particleUpdateScript [
   &FN_UpdateAnimTrail(SkeletalMesh=$System.SkeletalMesh)
   &ScaleRibbonWidth(Ribbon Width Scale=&FloatFromCurve(FloatCurve=DataInterface() {propertyValues {Curve {keys [
         {arriveTangent=32.7595, interpMode=RCIM_Cubic, leaveTangent=32.7595, tangentMode=RCTM_User, time=0, value=0}
         <at 0.1 1 RCIM_Cubic>
         {arriveTangent=-1.4352, tangentMode=RCTM_User, time=1, value=0}
        ]}}}))
   &ScaleColor(
    Scale Alpha=&Multiply_Float(
     A=&FloatFromCurve(FloatCurve=DataInterface() {propertyValues {Curve {keys [{arriveTangent=-0.5296, interpMode=RCIM_Cubic, leaveTangent=-0.5296, tangentMode=RCTM_User, time=0, value=0.25}, <at 1 0>]}}})
     B=1
    )
    Scale RGB=&VectorFromFloat(Value=$System.ToD Intensity)
    ScaleA=true
    ScaleRGB=true
   )
   &ParticleState()
  ]
  rendererClasses [&NiagaraRibbonRendererProperties]
  renderers [
   {rendererClass=&NiagaraRibbonRendererProperties}
   {rendererClass=&NiagaraRibbonRendererProperties, rendererIndex=1}
   {rendererClass=&NiagaraRibbonRendererProperties, rendererIndex=2}
  ]
 }
 Em_Trail002 {
  bEnabled=false
  emitterUpdateScript [
   &EmitterState()
   &FN_SpawnAnimTrail(Number of Anim Trails=3, Spawn Interval=0, Spawn Mode:ENiagaraSkeletalSpawning=::NewEnumerator0) {module=FN_SpawnAnimTrail001}
  ]
  particleSpawnScript [
   /Game/Effects/Niagara/Modules/AnimTrail/FN_InitializeAnimTrail/FN_InitializeAnimTrailBoneIndecies()
   /Game/Effects/Niagara/Modules/AnimTrail/FN_InitializeAnimTrailBoneIndecies_v2/FN_InitializeAnimTrailBoneIndecies(AnimTrail Selection=1, First Bone or Socket Index=1) {module=FN_InitializeAnimTrailBoneIndecies001}
   &FN_InitializeAnimTrailBoneIndecies(AnimTrail Selection=2, First Bone or Socket Index=2, Second Bone or Socket Index=2) {module=FN_InitializeAnimTrailBoneIndecies002}
   &SkeletalMeshLocation(
    Bone / Socket Index=$Particles.SpawnGroup
    Bone Sampling Mode=::NewEnumerator4
    MaskSpawnGroup=false
    Skeletal Mesh=$System.SkeletalMesh
    Velocity Sampling=::NewEnumerator0
    VelocitySpeedLimit=false
   )
   &FN_InitializeAnimTrail(
    AnimTrailWidth=2
    Color=&MultiplyLinearColor(
     Linear Color=<rgba 1 0.8509 0.5877 1>
     Scale Factor=&MakeVector4_FromLinearColor(LinearColor=$System.ToD Color)
    )
    Lifetime=1
    SkeletalMesh=$System.SkeletalMesh
   )
   /SetVariables(Particles.Position=&Add_Vector(A=Linked() {linkedVariable {name=Particles.Position, type=NiagaraPosition}}, B=<0 0 0>)) {isSetParameters=true, module=SetVariables_#4}
   /SetVariables(
    Particles.PositionOffset1:NiagaraPosition=&AddVectorToPosition(Position=$Particles.Position, Vector=<0 0 8>)
    Particles.PositionOffset2:NiagaraPosition=&AddVectorToPosition(Position=$Particles.Position, Vector=<0 0 -8>)
   ) {isSetParameters=true, module=SetVariables_#5}
   &AddVelocity(Velocity=<-1 0 1>) {enabled=false}
   &ApplyInitialForces() {enabled=false}
  ]
  particleUpdateScript [
   &FN_UpdateAnimTrail(SkeletalMesh=$System.SkeletalMesh)
   &ScaleRibbonWidth(Ribbon Width Scale=&FloatFromCurve(FloatCurve=DataInterface() {propertyValues {Curve {keys [
         {arriveTangent=32.7595, interpMode=RCIM_Cubic, leaveTangent=32.7595, tangentMode=RCTM_User, time=0, value=0}
         <at 0.1 1 RCIM_Cubic>
         {arriveTangent=-1.4352, tangentMode=RCTM_User, time=1, value=0}
        ]}}}))
   &ScaleColor(
    Scale Alpha=&Multiply_Float(
     A=&FloatFromCurve(FloatCurve=DataInterface() {propertyValues {Curve {keys [{arriveTangent=-0.5296, interpMode=RCIM_Cubic, leaveTangent=-0.5296, tangentMode=RCTM_User, time=0, value=0.25}, <at 1 0>]}}})
     B=1
    )
    Scale RGB=&VectorFromFloat(Value=$System.ToD Intensity)
    ScaleA=true
    ScaleRGB=true
   )
   &ParticleState()
  ]
  rendererClasses [&NiagaraRibbonRendererProperties]
  renderers [
   {rendererClass=&NiagaraRibbonRendererProperties}
   {rendererClass=&NiagaraRibbonRendererProperties, rendererIndex=1}
   {rendererClass=&NiagaraRibbonRendererProperties, rendererIndex=2}
  ]
 }
}
summary {
 systemName=NS_Pickaxe_TroutWrist_Emote
 emitters [
  {emitterName=Em_Trail, rendererClasses [&NiagaraRibbonRendererProperties]}
  {bEnabled=false, emitterName=Em_Trail001, rendererClasses [&NiagaraRibbonRendererProperties]}
  {bEnabled=false, emitterName=Em_Trail002, rendererClasses [&NiagaraRibbonRendererProperties]}
 ]
 userVariables [
  User.Ribbon camera face {type=Vector3f, description="", defaultValue {struct=Vector3f, value=<-0.1 0 0.1>}}
  User.RibbonIDOffset {type=NiagaraInt32}
  User.SwingSectionIndex {type=NiagaraInt32}
 ]
}
